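# Translucent UI material for an error message icon in Fortnite's in-game Creative UI.
# It blends between the ColorLeft and ColorRight parameters, both dark reds by default, along the horizontal normalized UI UV to form a left-to-right gradient.
# A texture sample supplies the error message icon alongside that gradient.
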Material /FortniteCoreUI/InGame/Creative/Art/M_UI_IconErrorMessage
graph {
 properties {
  bUsedWithStaticMesh=false
  blendMode=BLEND_Translucent
  materialDomain=MD_UI
  pixelDepthOffsetMode=PDOM_Legacy
  usedShadingModels=Default Lit
 }
 nodes [
  LinearInterpolate(
   A=VectorParameter() {parameterName=ColorLeft, defaultValue=<rgba 0.1 0 0 0>}.RGB
   B=VectorParameter() {parameterName=ColorRight, defaultValue=<rgba 0.5 0 0 0>}.RGB
   Alpha=ComponentMask(None=MaterialFunctionCall() {materialFunction=&GetUserInterfaceUV}.Normalized UV) {r=true}
  )
  TextureSample() {texture=&/T_icon_errormessage}
 ]
}
