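# Creative HUD element that shows one of three persistence warning banners: save error, local-only saving, or a save error advising island re-publish.
# On construct it binds to the player state's persistence error change event and applies the current value, unbinding on destruct.
# Each change collapses the active banner and selects the new one.
# Banners require persistence devices and a Published or Featured experience, with LSRError also allowed on Island volumes.
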WidgetBlueprint /FortniteCoreUI/InGame/Creative/HudWidgets/CreativePersistenceMessages
graphs [
 BindOrUnbindToPersistenceModeChange {nodes [
   BindOrUnbindToPersistenceModeChange!() -> (then, Bind:bool=~) {is=title K2Node_FunctionEntry}
   n1 = GetOwningPlayerState!() -> (FortPlayerStateAthena:Fort Player State Athena Object Reference)
   > Is Valid(InputObject=@n1)
   > n2 = Branch(Condition=Get Bind() -> (Bind:bool))
   n3 = Create Event()
   > Bind Event to On Persistence Error Type Changed(self:Fort Player State Athena Object Reference=@n1, Delegate=@n3) {is=title K2Node_AddDelegate}
   Unbind Event from On Persistence Error Type Changed(execute=@n2.else, self:Fort Player State Athena Object Reference=@n1, Delegate=@n3) {is=title K2Node_RemoveDelegate}
  ]}
 CanShowPersistenceError {nodes [
   n1 = CanShowPersistenceError!() -> (then, ContextActor:"Actor Object Reference (by ref)") {is=title K2Node_FunctionEntry}
   > n2 = GetComponentFromCurrentMinigameExperience!(
    execute
    ContextActor:Actor Object Reference=@n1
    ComponentClass:Actor Component Class Reference=/Script/FortniteGame.FortVolumePersistenceOptions
   ) -> (then, ReturnValue:Actor Component Object Reference)
   > n3 = Cast To FortVolumePersistenceOptions!(execute, Object=@n2) -> (then, CastFailed, AsFort Volume Persistence Options:Fort Volume Persistence Options Object Reference) {is=title K2Node_DynamicCast}
   > n4 = Is Valid(InputObject=@n3)
   Return Node!(execute=@n4.Is Not Valid, IsPersistenceEnabled:bool=false)
   > n5 = GetExperienceType!(execute, ExperienceContextActor:Actor Object Reference=@n1) -> (then, ReturnValue:EFortVolumeType Enum=None)
   n6 = AND Boolean(
    A="Equal (Enum)"(A:EFortVolumeType Enum=@n5, B:EFortVolumeType Enum=Island)
    B="Equal (Object)"(A=Get CurrentActiveWidget() -> (CurrentActiveWidget:Widget Object Reference), B=Get LSRError() -> (LSRError:Border Object Reference))
   )
   n7 = "Equal (Enum)"(A=@n5, B=Published)
   n8 = "Equal (Enum)"(A=@n5, B=Featured)
   > Return Node(IsPersistenceEnabled=AND Boolean!(
     A=OR Boolean!(A=@n7, B=@n8, C:bool=@n6) -> (ReturnValue)
     B=IsValid(Object=@n3)
     C:bool=IsTherePersistenceDevices!(self:Fort Volume Persistence Options Object Reference=@n3) -> (ReturnValue:bool)
    ) -> (ReturnValue))
  ]}
 EventGraph {nodes [
   Event Construct()
   > BindOrUnbindToPersistenceModeChange!(execute, self, Bind:bool=true) -> (then) {is=title K2Node_CallFunction}
   > OnPersistenceErrorTypeChanged!(
    execute
    self
    FortPlayerState:Fort Player State Athena Object Reference
    NewValue:ECreativePersistenceErrorTypes Enum=GetCurrentPersistenceErrorType!(self:Fort Player State Athena Object Reference=GetOwningPlayerState()) -> (ReturnValue:ECreativePersistenceErrorTypes Enum=None)
   ) -> (then)
   Event Destruct()
   > BindOrUnbindToPersistenceModeChange(Bind=false)
  ]}
 GetOwningPlayerController {nodes [
   GetOwningPlayerController() {is=title K2Node_FunctionEntry}
   > n1 = Is Valid(InputObject=Get CachedController() -> (CachedController:Fort Player Controller Athena Object Reference))
   n2 = Cast To FortPlayerControllerAthena(execute=@n1.Is Not Valid, Object=GetOwningPlayer())
   > n3 = Set CachedController(CachedController:Fort Player Controller Athena Object Reference=@n2) -> (Output_Get:Fort Player Controller Athena Object Reference)
   Return Node!(execute=[@n3.then, @n1.Is Valid], FortPCAthena:Fort Player Controller Athena Object Reference=Get CachedController())
  ]}
 GetOwningPlayerState {nodes [
   GetOwningPlayerState() {is=title K2Node_FunctionEntry}
   > n1 = Is Valid(InputObject=Get CachedPlayerState() -> (CachedPlayerState:Fort Player State Athena Object Reference))
   n2 = Cast To FortPlayerStateAthena!(
    execute=@n1.Is Not Valid
    Object=Get PlayerState!(self:Controller Object Reference=GetOwningPlayerController!() -> (FortPCAthena:Fort Player Controller Athena Object Reference) {is=title K2Node_CallFunction}) -> (PlayerState:Player State Object Reference)
   ) -> (then, CastFailed, AsFort Player State Athena:Fort Player State Athena Object Reference) {is=title K2Node_DynamicCast}
   > n3 = Set CachedPlayerState(CachedPlayerState:Fort Player State Athena Object Reference=@n2) -> (Output_Get:Fort Player State Athena Object Reference)
   Return Node!(execute=[@n3.then, @n1.Is Valid], FortPlayerStateAthena:Fort Player State Athena Object Reference=Get CachedPlayerState())
  ]}
 OnPersistenceErrorTypeChanged {nodes [
   n1 = OnPersistenceErrorTypeChanged!() -> (then, FortPlayerState:Fort Player State Athena Object Reference, NewValue:ECreativePersistenceErrorTypes Enum=None) {is=title K2Node_FunctionEntry}
   > n2 = Get!(execute:Exec) -> (then:Exec, else:Exec, CurrentActiveWidget:Widget Object Reference)
   > n3 = Set Message Visibility!(execute, self, IsVisible:bool=false, Widget:Widget Object Reference=@n2) -> (then) {is=title K2Node_CallFunction}
   n4 = Branch(
    execute=[@n3.then, @n2.else]
    Condition="Equal (Enum)"(A:ECreativePersistenceErrorTypes Enum=@n1.NewValue, B:ECreativePersistenceErrorTypes Enum=None)
   )
   n5 = Set CurrentActiveWidget(
    execute=@n4.else
    CurrentActiveWidget:Widget Object Reference=Select!(
     None:Border Object Reference
     VersionError:Border Object Reference=Get ErrorBorder() -> (ErrorBorder:Border Object Reference)
     LocalOnly:Border Object Reference=Get LocalOnlyBorder() -> (LocalOnlyBorder:Border Object Reference)
     LSRError:Border Object Reference=Get LSRError()
     Index:ECreativePersistenceErrorTypes Enum=@n1.NewValue
    ) -> (ReturnValue:Border Object Reference)
   ) -> (Output_Get:Widget Object Reference)
   > Set Message Visibility(IsVisible=true, Widget=@n5)
  ]}
 Set Message Visibility {nodes [
   n1 = Set Message Visibility!() -> (then, IsVisible:bool=~, Widget:Widget Object Reference) {is=title K2Node_FunctionEntry}
   n2 = GetOwningPlayerController()
   > n3 = Is Valid(InputObject=@n2)
   > n4 = Branch(Condition=CanShowPersistenceError!(self, ContextActor:"Actor Object Reference (by ref)"=@n2) -> (IsPersistenceEnabled:bool) {is=title K2Node_CallFunction})
   > n5 = Branch(Condition=@n1)
   > SetVisibility(self=Get Widget() -> (Widget:Widget Object Reference), InVisibility=SelfHitTestInvisible)
   SetVisibility(execute=[@n5.else, @n4.else, @n3.Is Not Valid], self=Get Widget(), InVisibility=Collapsed)
  ]}
]
variables [
 CachedController {type=FortPlayerControllerAthena, default=null}
 CachedPlayerState {type=FortPlayerStateAthena, default=null}
 CurrentActiveWidget {type=Widget, default=null}
]
widgets {
 info {inheritedWidgetCount=1, parentClass=&FortHUDElementWidget, rootWidgetClass=&VerticalBox, widgetCount=14}
 widgets [
  VerticalBox_0 {widgetClassPath=&VerticalBox, widget=VerticalBox_0}
  ErrorBorder {
   widgetClassPath=&Border
   bIsVariable=true
   parent=VerticalBox_0
   widget=ErrorBorder
   properties {
    visibility=Collapsed
    background {
     bIsDynamicallyLoaded=false
     drawAs=Box
     imageType=NoImage
     mirroring=NoMirror
     resourceName=None
     resourceObject=/FortniteCoreUI/InGame/Creative/Art/M_UI_IconErrorMessage
     tiling=NoTile
     imageSize=<32 32>
     margin=<ltrb 0.5 0.5 0.5 0.5>
     outlineSettings {
      bUseBrushTransparency=false
      roundingType=HalfHeightRadius
      width=0
      color {colorUseRule=UseColor_Specified, specifiedColor=<rgba 0 0 0 0>}
      cornerRadii=<0 0 0 0>
     }
     tintColor {colorUseRule=UseColor_Specified, specifiedColor=<rgba 1 1 1 1>}
     uVRegion {bIsValid=false, max=<0 0>, min=<0 0>}
    }
    padding=<ltrb 0 4 0 4>
   }
   slot {class=VerticalBoxSlot}
  }
  HorizontalBox_0 {
   widgetClassPath=&HorizontalBox
   parent=ErrorBorder
   widget=HorizontalBox_0
   slot {class=BorderSlot, properties {padding=<ltrb 0 4 0 4>}}
  }
  Icon_Warning {
   widgetClassPath=&Image
   bIsVariable=true
   parent=HorizontalBox_0
   widget=Icon_Warning
   properties {brush {
     bIsDynamicallyLoaded=false
     drawAs=Image
     imageType=NoImage
     mirroring=NoMirror
     resourceName=None
     resourceObject=/Game/UI/Foundation/Shell/Textures/T-Icon-Warning-32
     tiling=NoTile
     imageSize=<32 32>
     margin=<ltrb 0 0 0 0>
     outlineSettings {
      bUseBrushTransparency=false
      roundingType=HalfHeightRadius
      width=0
      color {colorUseRule=UseColor_Specified, specifiedColor=<rgba 0 0 0 0>}
      cornerRadii=<0 0 0 0>
     }
     tintColor {colorUseRule=UseColor_Specified, specifiedColor=<rgba 1 1 1 1>}
     uVRegion {bIsValid=false, max=<0 0>, min=<0 0>}
    }}
   slot {class=HorizontalBoxSlot, properties {padding=<ltrb 20 0 0 0>}}
  }
  TextCreativeMnemonic {
   widgetClassPath=&TextBlock
   bIsVariable=true
   parent=HorizontalBox_0
   widget=TextCreativeMnemonic
   properties {
    justification=Center
    text="Save Device Error - Progress will not be saved!"
    font {
     bForceMonospaced=false
     bMaterialIsStencil=false
     fontMaterial=null
     fontObject=&NotoSans
     letterSpacing=0
     monospacedWidth=1
     size=16
     skewAmount=0
     typefaceFontName=NotoSans-Bold
     outlineSettings {
      bApplyOutlineToDropShadows=false
      bMiteredCorners=false
      bSeparateFillAlpha=false
      outlineMaterial=null
      outlineSize=0
      outlineColor=<rgba 0 0 0 1>
     }
    }
   }
   slot {class=HorizontalBoxSlot, properties {horizontalAlignment=HAlign_Center, verticalAlignment=VAlign_Center, padding=<ltrb 10 0 20 0>}}
  }
  LocalOnlyBorder {
   widgetClassPath=&Border
   bIsVariable=true
   parent=VerticalBox_0
   widget=LocalOnlyBorder
   properties {
    visibility=Collapsed
    background {
     bIsDynamicallyLoaded=false
     drawAs=Box
     imageType=NoImage
     mirroring=NoMirror
     resourceName=None
     resourceObject=/FortniteCoreUI/InGame/Creative/Art/MI_UI_IconErrorMessageLocalPersistence
     tiling=NoTile
     imageSize=<32 32>
     margin=<ltrb 0.5 0.5 0.5 0.5>
     outlineSettings {
      bUseBrushTransparency=false
      roundingType=HalfHeightRadius
      width=0
      color {colorUseRule=UseColor_Specified, specifiedColor=<rgba 0 0 0 0>}
      cornerRadii=<0 0 0 0>
     }
     tintColor {colorUseRule=UseColor_Specified, specifiedColor=<rgba 1 1 1 1>}
     uVRegion {bIsValid=false, max=<0 0>, min=<0 0>}
    }
    padding=<ltrb 0 4 0 4>
   }
   slot {class=VerticalBoxSlot}
  }
  HorizontalBox {
   widgetClassPath=&HorizontalBox
   parent=LocalOnlyBorder
   widget=HorizontalBox
   slot {class=BorderSlot, properties {padding=<ltrb 0 4 0 4>}}
  }
  Icon_Warning_1 {
   widgetClassPath=&Image
   bIsVariable=true
   parent=HorizontalBox
   widget=Icon_Warning_1
   properties {brush {
     bIsDynamicallyLoaded=false
     drawAs=Image
     imageType=NoImage
     mirroring=NoMirror
     resourceName=None
     resourceObject=&T-Icon-Warning-32
     tiling=NoTile
     imageSize=<32 32>
     margin=<ltrb 0 0 0 0>
     outlineSettings {
      bUseBrushTransparency=false
      roundingType=HalfHeightRadius
      width=0
      color {colorUseRule=UseColor_Specified, specifiedColor=<rgba 0 0 0 0>}
      cornerRadii=<0 0 0 0>
     }
     tintColor {colorUseRule=UseColor_Specified, specifiedColor=<rgba 1 1 1 1>}
     uVRegion {bIsValid=false, max=<0 0>, min=<0 0>}
    }}
   slot {class=HorizontalBoxSlot, properties {padding=<ltrb 20 0 0 0>}}
  }
  TextCreativeMnemonic_1 {
   widgetClassPath=&TextBlock
   bIsVariable=true
   parent=HorizontalBox
   widget=TextCreativeMnemonic_1
   properties {
    justification=Center
    text="Playing in a non active Island version - progress will only be saved locally!"
    font {
     bForceMonospaced=false
     bMaterialIsStencil=false
     fontMaterial=null
     fontObject=&NotoSans
     letterSpacing=0
     monospacedWidth=1
     size=16
     skewAmount=0
     typefaceFontName=NotoSans-Bold
     outlineSettings {
      bApplyOutlineToDropShadows=false
      bMiteredCorners=false
      bSeparateFillAlpha=false
      outlineMaterial=null
      outlineSize=0
      outlineColor=<rgba 0 0 0 1>
     }
    }
   }
   slot {class=HorizontalBoxSlot, properties {horizontalAlignment=HAlign_Center, verticalAlignment=VAlign_Center, padding=<ltrb 10 0 20 0>}}
  }
  LSRError {
   widgetClassPath=&Border
   bIsVariable=true
   parent=VerticalBox_0
   widget=LSRError
   properties {
    visibility=Collapsed
    background {
     bIsDynamicallyLoaded=false
     drawAs=Box
     imageType=NoImage
     mirroring=NoMirror
     resourceName=None
     resourceObject=&MI_UI_IconErrorMessageLocalPersistence
     tiling=NoTile
     imageSize=<32 32>
     margin=<ltrb 0.5 0.5 0.5 0.5>
     outlineSettings {
      bUseBrushTransparency=false
      roundingType=HalfHeightRadius
      width=0
      color {colorUseRule=UseColor_Specified, specifiedColor=<rgba 0 0 0 0>}
      cornerRadii=<0 0 0 0>
     }
     tintColor {colorUseRule=UseColor_Specified, specifiedColor=<rgba 1 1 1 1>}
     uVRegion {bIsValid=false, max=<0 0>, min=<0 0>}
    }
    padding=<ltrb 0 4 0 4>
   }
   slot {class=VerticalBoxSlot}
  }
  HorizontalBox_1 {
   widgetClassPath=&HorizontalBox
   parent=LSRError
   widget=HorizontalBox_1
   slot {class=BorderSlot, properties {padding=<ltrb 0 4 0 4>}}
  }
  Icon_Warning_2 {
   widgetClassPath=&Image
   bIsVariable=true
   parent=HorizontalBox_1
   widget=Icon_Warning_2
   properties {brush {
     bIsDynamicallyLoaded=false
     drawAs=Image
     imageType=NoImage
     mirroring=NoMirror
     resourceName=None
     resourceObject=&T-Icon-Warning-32
     tiling=NoTile
     imageSize=<32 32>
     margin=<ltrb 0 0 0 0>
     outlineSettings {
      bUseBrushTransparency=false
      roundingType=HalfHeightRadius
      width=0
      color {colorUseRule=UseColor_Specified, specifiedColor=<rgba 0 0 0 0>}
      cornerRadii=<0 0 0 0>
     }
     tintColor {colorUseRule=UseColor_Specified, specifiedColor=<rgba 1 1 1 1>}
     uVRegion {bIsValid=false, max=<0 0>, min=<0 0>}
    }}
   slot {class=HorizontalBoxSlot, properties {padding=<ltrb 20 0 0 0>}}
  }
  TextCreativeMnemonic_2 {
   widgetClassPath=&TextBlock
   bIsVariable=true
   parent=HorizontalBox_1
   widget=TextCreativeMnemonic_2
   properties {
    justification=Center
    text="Save Device Error - Contact the Creator, as re-publishing the Island may fix the problem"
    font {
     bForceMonospaced=false
     bMaterialIsStencil=false
     fontMaterial=null
     fontObject=&NotoSans
     letterSpacing=0
     monospacedWidth=1
     size=16
     skewAmount=0
     typefaceFontName=NotoSans-Bold
     outlineSettings {
      bApplyOutlineToDropShadows=false
      bMiteredCorners=false
      bSeparateFillAlpha=false
      outlineMaterial=null
      outlineSize=0
      outlineColor=<rgba 0 0 0 1>
     }
    }
   }
   slot {class=HorizontalBoxSlot, properties {horizontalAlignment=HAlign_Center, verticalAlignment=VAlign_Center, padding=<ltrb 10 0 20 0>}}
  }
  TutorialHighlight {widgetClassPath=&TutorialHighlight}
 ]
}
