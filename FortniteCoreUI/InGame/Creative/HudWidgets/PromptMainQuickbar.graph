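# Creative-mode HUD element tagged HUD.Creative.MainQuickBar that shows a single quickbar slot prompt.
# A FortHUDElementWrapper root hosts a vertical box stacking the slot, a hidden uppercase placeholder message and a keybind widget.
# The slot supports the Death cooldown type.
# The Blueprint is layout only, with an empty event graph and no variables of its own beyond the exposed widgets.
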WidgetBlueprint /FortniteCoreUI/InGame/Creative/HudWidgets/PromptMainQuickbar
graphs [EventGraph {links [], nodes []}]
variables []
widgets {
 info {
  inheritedWidgetCount=1
  namedSlotCount=1
  parentClass=&FortHUDElementWidget
  rootWidgetClass=/Game/Athena/HUD/FortHUDElementWrapper.FortHUDElementWrapper_C
  widgetCount=6
 }
 widgets [
  FortHUDElementWrapper {
   widgetClassPath=/Game/Athena/HUD/FortHUDElementWrapper
   bIsVariable=true
   widget=FortHUDElementWrapper
   properties {elementTag {tagName=HUD.Creative.MainQuickBar}}
  }
  VerticalBox_0 {widgetClassPath=&VerticalBox, namedSlotHost=FortHUDElementWrapper, widget=VerticalBox_0}
  QuickbarSlot1 {
   widgetClassPath=/Game/Athena/HUD/AthenaQuickBarSlot
   bIsVariable=true
   parent=VerticalBox_0
   widget=QuickbarSlot1
   properties {cooldownTypesSupported [Death]}
   slot {class=VerticalBoxSlot, properties {padding=<ltrb 0 0 8 0>}}
  }
  Text_Message {
   widgetClassPath=&CommonTextBlock
   parent=VerticalBox_0
   widget=Text_Message
   properties {
    text=PLACEHOLDER TEXT
    textTransformPolicy=ToUpper
    visibility=Hidden
    font {
     bForceMonospaced=false
     bMaterialIsStencil=false
     fontMaterial=null
     fontObject=&Burbank
     letterSpacing=0
     monospacedWidth=1
     size=18
     skewAmount=0
     typefaceFontName=BurbankRegular-Black
     outlineSettings {
      bApplyOutlineToDropShadows=false
      bMiteredCorners=false
      bSeparateFillAlpha=false
      outlineMaterial=null
      outlineSize=0
      outlineColor=<rgba 0 0 0 1>
     }
    }
    shadowColorAndOpacity=<rgba 0.0052 0.0052 0.0056 0.822>
    shadowOffset=<3 3>
   }
   slot {class=VerticalBoxSlot, properties {horizontalAlignment=HAlign_Center, verticalAlignment=VAlign_Center}}
  }
  ClutchKeybindWidget {
   widgetClassPath=/FortUILibrary/Input/KeybindWidget
   bIsVariable=true
   parent=VerticalBox_0
   widget=ClutchKeybindWidget
   slot {class=VerticalBoxSlot, properties {horizontalAlignment=HAlign_Center, verticalAlignment=VAlign_Center}}
  }
  TutorialHighlight {widgetClassPath=&TutorialHighlight}
 ]
}
